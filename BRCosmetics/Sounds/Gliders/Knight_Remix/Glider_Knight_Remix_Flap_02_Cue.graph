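# Sound cue for the Knight Remix glider's wing flap sound, variant 02.
# An equally weighted random node picks one of five flap waves, and a modulator plays it well below normal pitch with slightly varied volume.
# It uses a glider thrust attenuation asset and the glider landing sound class at reduced cue volume.
# Overridden concurrency limits it to two sounds per owner, stopping the oldest.
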SoundCue /BRCosmetics/Sounds/Gliders/Knight_Remix/Glider_Knight_Remix_Flap_02_Cue
root=Modulator_0
nodes [
 {id=Modulator_0, children [Random_0], properties {pitchMax=0.75, pitchMin=0.7, volumeMin=0.8}}
 {id=Random_0, children [WavePlayer_1, WavePlayer_2, WavePlayer_3, WavePlayer_9, WavePlayer_11], properties {weights [1,1,1,1,1]}}
 {id=WavePlayer_1, properties {soundWaveAssetPtr=&/Glider_Knight_Remix_Wing_Flap_01}}
 {id=WavePlayer_11, properties {soundWaveAssetPtr=&/Glider_Knight_Remix_Wing_Flap_06}}
 {id=WavePlayer_2, properties {soundWaveAssetPtr=&/Glider_Knight_Remix_Wing_Flap_02}}
 {id=WavePlayer_3, properties {soundWaveAssetPtr=&/Glider_Knight_Remix_Wing_Flap_03}}
 {id=WavePlayer_9, properties {soundWaveAssetPtr=&/Glider_Knight_Remix_Wing_Flap_04}}
]
properties {
 attenuationSettings=&Fort_Player_Glider_Thrust_Athena_Att
 bOverrideConcurrency=true
 maxDistance=12500
 soundClassObject=&Fort_Gliders_Land
 volumeMultiplier=0.3
 concurrencyOverrides {bLimitToOwner=true, maxCount=2, resolutionRule=StopOldest}
 concurrencySet [&Athena_GliderOpenClose_Conc]
}
