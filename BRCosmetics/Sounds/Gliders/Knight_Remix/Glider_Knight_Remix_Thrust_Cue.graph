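# Sound cue for the thrust sound of the Knight Remix glider cosmetic.
# A Random node picks one of three equally weighted thrust waves, then a Modulator varies its pitch within a range below normal.
# It plays at reduced volume with player glider thrust attenuation and a glider loop audio properties sheet.
# It references a thrust concurrency set and carries a single-instance override that prevents new plays.
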SoundCue /BRCosmetics/Sounds/Gliders/Knight_Remix/Glider_Knight_Remix_Thrust_Cue
root=Modulator_0
nodes [
 {id=Modulator_0, children [Random_1], properties {pitchMax=0.9, pitchMin=0.7}}
 {id=Random_1, children [WavePlayer_1,WavePlayer_5,WavePlayer_6], properties {weights [1,1,1]}}
 {id=WavePlayer_1, properties {soundWaveAssetPtr=&/Glider_Knight_Remix_Thrust_01}}
 {id=WavePlayer_5, properties {soundWaveAssetPtr=&/Glider_Knight_Remix_Thrust_02}}
 {id=WavePlayer_6, properties {soundWaveAssetPtr=&/Glider_Knight_Remix_Thrust_03}}
]
properties {
 attenuationSettings=&Fort_Player_Glider_Thrust_Athena_Att
 audioPropertiesSheet=&APS_Cosmetic_Glider_Loop
 maxDistance=12500
 soundClassObject=&Fort_Gliders_Land
 volumeMultiplier=0.2
 concurrencyOverrides {maxCount=1, resolutionRule=PreventNew}
 concurrencySet [&Athena_Glider_Thrust_Conc]
}
